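; Prototype sliding double door, built on a parent door class, that closes after a delay with an optional spinning alarm light and crushes what it catches.
; Timelines drive the door percentage, which lerps both panels and enables blocking collision at full close.
; Closing damages KillBox players lethally and applies a damage GameplayEffect to vehicles.
; Push components launch targeted players sideways on authority, though only the right one's async targeting is started here.
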(event EventBeginPlay
  (|CallParentFunction)
  (Variables|/Script/DragonCartDoorsRuntime.DragonCartDoorPushComponent|SetPushPointSceneComp (Variables|Default|GetPushPoint_L) (Variables|Default|GetDragonCartDoorPush_L))
  (Variables|/Script/DragonCartDoorsRuntime.DragonCartDoorPushComponent|SetPushPointSceneComp (Variables|Default|GetPushPoint_R) (Variables|Default|GetDragonCartDoorPush_R)))

(event EventBP_OnActive
  (params (EnableAlternateActive :type Boolean) (EnableOpenCues :type Boolean))
  (|CallParentFunction EnableAlternateActive EnableOpenCues)
  (if (Variables|Default|GetPlayAlarm)
    (CallFunction|EnableAlarm)
    (Utilities|FlowControl|Sequence
      (:then_0
        _)
      (:then_1
        (bind (_direction _rotation) (Timeline|AlarmLightRotator PlayFromStart)
          (:Update
            (Variables|Default|SetAlarmLightRotationY _rotation)
            (CallFunction|RotateAlarmLights)))))
    (else
      _))
  (Utilities|FlowControl|Delay (Variables|Default|GetCloseDelay))
  (Class|DragonCartDoorPushComponent|StartImpulseAsyncTargeting (Variables|Default|GetDragonCartDoorPush_R))
  (bind (_direction_2 _alpha) (Timeline|TL_CloseDoor PlayFromStart)
    (:Update
      (Variables|Default|SetDoorOperationPercentage _alpha))
    (:Finished
      (CallFunction|SmashPlayer)
      (Networking|SwitchHasAuthority
        (:Authority
          (Variables|Default|SetDoorState "NewEnumerator2")
          (Default|CallOnDoorFinishActivation)
          (Class|DragonCartDoorPushComponent|StopImpulseAsyncTargeting (Variables|Default|GetDragonCartDoorPush_R)))))
    (:DestroyVehicles
      (CallFunction|SmashCar))))

(event EventBP_OnClosed
  (params (EnableAlternateClose :type Boolean) (EnableOpenCues :type Boolean))
  (|CallParentFunction EnableAlternateClose EnableOpenCues)
  (if (Variables|Default|GetPlayAlarm)
    (CallFunction|DisableAlarm)))

(event Default|EventBP_OnReset
  (params (EnableAlternateReset :type Boolean) (EnableOpenCues :type Boolean))
  (|CallParentFunction EnableAlternateReset EnableOpenCues)
  (bind (_direction _alpha) (Timeline|TL_Open_Door PlayFromStart)
    (:Update
      (Variables|Default|SetDoorOperationPercentage _alpha))
    (:Finished
      (Networking|SwitchHasAuthority
        (:Authority
          (Variables|Default|SetDoorState "NewEnumerator0")
          (Default|CallOnDoorFinishReset))))))

(event Custom|SmashPlayer
  (for _array_element (Collision|GetOverlappingActors (Variables|Default|GetKillBox) "/Script/FortniteGame.FortPlayerPawn" :!class /Script/Engine.PrimitiveComponent)
    (bind _asfort_player_pawn (Utilities|Casting|CastToFortPlayerPawn _array_element))
    (Game|Damage|ApplyDamage _asfort_player_pawn 4444.0 0 0 "/Engine/EngineDamageTypes/DmgTypeBP_Environmental.DmgTypeBP_Environmental_C")))

(event Custom|SmashCar
  (for _array_element (Collision|GetOverlappingActors (Variables|Default|GetKillBox) 0 :!class /Script/Engine.PrimitiveComponent)
    (bind _asfort_athena_vehicle (Utilities|Casting|CastToFortAthenaVehicle _array_element))
    (Fortnite|ApplyGameplayEffectsToActor _asfort_athena_vehicle (Utilities|Array|MakeArray (Variables|Default|GetVehicleDamageGE_0)) (Transformation|GetActorLocation _asfort_athena_vehicle) false 0 (Ability|GetAbilitySystemComponent 0))))

(event OnLinkedActorDestroyed(FortLinkToActor)
  (params (DamageCauser :type Actor%20Object%20Reference))
  (Rendering|SetActorHiddenInGame true)
  (Actor|DestroyActor))

(event OnActorTargeted(DragonCartDoorPush_L)
  (params (TargetedActor :type Actor%20Object%20Reference))
  (Networking|SwitchHasAuthority
    (:Authority
      (bind _asfort_player_pawn (Utilities|Casting|CastToFortPlayerPawn TargetedActor))
      (Character|LaunchCharacter _asfort_player_pawn (* (Transformation|GetRightVector (Variables|Default|GetPushPoint_L)) (ScalableFloat|GetValueatLevel (Variables|Default|GetHF_PushAwayStrength))) false false))))

(event OnActorTargeted(DragonCartDoorPush_R)
  (params (TargetedActor :type Actor%20Object%20Reference))
  (Networking|SwitchHasAuthority
    (:Authority
      (bind _asfort_player_pawn (Utilities|Casting|CastToFortPlayerPawn TargetedActor))
      (Character|LaunchCharacter _asfort_player_pawn (* (Transformation|GetRightVector (Variables|Default|GetPushPoint_R)) (ScalableFloat|GetValueatLevel (Variables|Default|GetHF_PushAwayStrength))) false false))))

(fn DetermineLaunchDirection
  (params (Actor :type Actor%20Object%20Reference))
  (Variables|Default|SetLaunchVelocity (Math|Vector|SelectVector (* (* (+ (Transformation|GetActorRightVector) (Math|Vector|MakeVector 0.0 0.0 (Variables|Default|GetLaunchBoostInitialZ))) (Variables|Default|GetLaunchPower)) (Math|Vector|MakeVector (* (.x (* (+ (Transformation|GetActorRightVector) (Math|Vector|MakeVector 0.0 0.0 (Variables|Default|GetLaunchBoostInitialZ))) (Variables|Default|GetLaunchPower))) (Variables|Default|GetLaunchBoostX)) (* (.y (* (+ (Transformation|GetActorRightVector) (Math|Vector|MakeVector 0.0 0.0 (Variables|Default|GetLaunchBoostInitialZ))) (Variables|Default|GetLaunchPower))) (Variables|Default|GetLaunchBoostY)) (* (.z (* (+ (Transformation|GetActorRightVector) (Math|Vector|MakeVector 0.0 0.0 (Variables|Default|GetLaunchBoostInitialZ))) (Variables|Default|GetLaunchPower))) (Variables|Default|GetLaunchBoostZ)))) (* (* (* (+ (Transformation|GetActorRightVector) (Math|Vector|MakeVector 0.0 0.0 (Variables|Default|GetLaunchBoostInitialZ))) (Variables|Default|GetLaunchPower)) (Math|Vector|MakeVector (* (.x (* (+ (Transformation|GetActorRightVector) (Math|Vector|MakeVector 0.0 0.0 (Variables|Default|GetLaunchBoostInitialZ))) (Variables|Default|GetLaunchPower))) (Variables|Default|GetLaunchBoostX)) (* (.y (* (+ (Transformation|GetActorRightVector) (Math|Vector|MakeVector 0.0 0.0 (Variables|Default|GetLaunchBoostInitialZ))) (Variables|Default|GetLaunchPower))) (Variables|Default|GetLaunchBoostY)) (* (.z (* (+ (Transformation|GetActorRightVector) (Math|Vector|MakeVector 0.0 0.0 (Variables|Default|GetLaunchBoostInitialZ))) (Variables|Default|GetLaunchPower))) (Variables|Default|GetLaunchBoostZ)))) "1.000000,-1.000000,1.000000") (<= (Math|Vector|Distance(Vector) (Transformation|GetActorLocation Actor) (Transformation|GetWorldLocation (Variables|Default|GetLightSpinnerFront))) (Math|Vector|Distance(Vector) (Transformation|GetActorLocation Actor) (Transformation|GetWorldLocation (Variables|Default|GetLightSpinnerBack))))))
  (return))

(fn DisableAlarm
  (Rendering|Material|SetMaterial (Variables|Default|GetAlarmLight) 0 "/DragonCartDoors/Assets/MI_FlatColorParent_Iron.MI_FlatColorParent_Iron")
  (Rendering|SetVisibility (Variables|Default|GetFortSpotLightFront1) false false)
  (return))

(fn EnableAlarm
  (Rendering|Material|SetMaterial (Variables|Default|GetAlarmLight) 0 "/DragonCartDoors/Assets/MI_FlatColorParent_Red.MI_FlatColorParent_Red")
  (Rendering|SetVisibility (Variables|Default|GetFortSpotLightFront1) true false)
  (return))

(fn MoveDoorByPercentage
  (params (Percentage :type Float%20%28double-precision%29))
  (|CallParentFunction Percentage)
  (Transformation|SetRelativeLocation (Variables|Default|GetPanel_L) (Math|Vector|MakeVector (Math|Float|Lerp (Variables|Default|GetLeftStartX) (Variables|Default|GetLeftEndX) Percentage) (.y (Variables|/Script/Engine.SceneComponent|GetRelativeLocation (Variables|Default|GetPanel_L))) (.z (Variables|/Script/Engine.SceneComponent|GetRelativeLocation (Variables|Default|GetPanel_L)))) false false)
  (Transformation|SetRelativeLocation (Variables|Default|GetPanel_R) (Math|Vector|MakeVector (Math|Float|Lerp (Variables|Default|GetRightStartX) (Variables|Default|GetRightEndX) Percentage) (.y (Variables|/Script/Engine.SceneComponent|GetRelativeLocation (Variables|Default|GetPanel_L))) (.z (Variables|/Script/Engine.SceneComponent|GetRelativeLocation (Variables|Default|GetPanel_L)))) false false)
  (Collision|SetCollisionEnabled (Variables|Default|GetDoorClosedCollision) (select (< Percentage 1.0) "NoCollision" "QueryAndPhysics")))

(fn RotateAlarmLights
  (Transformation|SetWorldRotation (Variables|Default|GetLightSpinnerFront) (Math|Rotator|MakeRotator (.roll (Transformation|GetWorldRotation (Variables|Default|GetLightSpinnerFront))) (Variables|Default|GetAlarmLightRotationY) (.yaw (Transformation|GetWorldRotation (Variables|Default|GetLightSpinnerFront)))) false false)
  (Transformation|SetWorldRotation (Variables|Default|GetLightSpinnerBack) (Math|Rotator|MakeRotator (.roll (Transformation|GetWorldRotation (Variables|Default|GetLightSpinnerBack))) (Variables|Default|GetAlarmLightRotationY) (.yaw (Transformation|GetWorldRotation (Variables|Default|GetLightSpinnerBack)))) false false)
  (return))

(fn ConstructionScript)
