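; Large three-panel Buckle Cabin door child Blueprint: side panels slide along X and the middle along Z, with an editor preview of partial opening.
; Activation plays the alternate timeline and reset reverses it, feeding the door percentage and firing temporary sound cues because parent cues fail.
; Push components drive server-side launches of targeted players, while a kill box smashes vehicles and players as the timeline reverses.
; It also registers and toggles vault alarms.
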(event EventBeginPlay
  (|CallParentFunction)
  (Variables|Default|SetBlockPartialOpen true)
  (Variables|Default|SetPanelTravel 0.0))

(event EventEndPlay
  (params (EndPlayReason :type EEndPlayReason%20Enum :default "Destroyed"))
  (|CallParentFunction EndPlayReason)
  (Variables|Default|SetBlockPartialOpen false)
  (Variables|Default|SetPanelTravel 0.0))

(event EventBP_OnActive
  (params (EnableAlternateActive :type Boolean) (EnableOpenCues :type Boolean))
  (|CallParentFunction EnableAlternateActive EnableOpenCues)
  (Variables|/Script/DragonCartDoorsRuntime.DragonCartDoorPushComponent|SetPushPointSceneComp (Variables|Default|GetPushPointL) (Variables|Default|GetDragonCartDoorPushL))
  (Variables|/Script/DragonCartDoorsRuntime.DragonCartDoorPushComponent|SetPushPointSceneComp (Variables|Default|GetPushPointR) (Variables|Default|GetDragonCartDoorPushR))
  (Variables|/Script/DragonCartDoorsRuntime.DragonCartDoorPushComponent|SetPushPointSceneComp (Variables|Default|GetPushPointM) (Variables|Default|GetDragonCartDoorPushM))
  (CallFunction|OnActiveAlternate)
  (Development|PrintString "Large Door - Alternate Open" true true "(R=1.000000,G=0.400000,B=0.000000,A=1.000000)" 5.0))

(event Default|EventBP_OnReset
  (params (EnableAlternateReset :type Boolean) (EnableOpenCues :type Boolean))
  (|CallParentFunction EnableAlternateReset EnableOpenCues)
  (CallFunction|OnResetAlternate)
  (Development|PrintString "Large Door - Alternate Reset" true true "(R=1.000000,G=0.400000,B=0.000000,A=1.000000)" 5.0))

(event EventRegisterVaultDoorAlarm
  (params (AlarmToRegister :type Actor%20Object%20Reference))
  (Utilities|Array|AddUnique (Variables|Default|GetVaultAlarms) AlarmToRegister))

(event EventBP_OnClosed
  (params (EnableAlternateClose :type Boolean) (EnableOpenCues :type Boolean))
  (|CallParentFunction EnableAlternateClose EnableOpenCues))

(event Custom|OnActiveNormal
  (CallFunction|PlayOpeningSFXTemp)
  (bind (_direction _alpha) (Timeline|TL_CloseDoor Play)
    (:Update
      (Variables|Default|SetDoorOperationPercentage _alpha))
    (:Finished
      (CallFunction|OnDoorTimelineFinished _direction))
    (:DestroyVehicles
      (Networking|SwitchHasAuthority
        (:Authority
          (CallFunction|SmashVehicles (CallFunction|GetVehiclestoSmash)))))))

(event Custom|OnResetNormal
  (CallFunction|PlayClosingSFXTemp)
  (Class|DragonCartDoorPushComponent|StartImpulseAsyncTargeting (Variables|Default|GetDragonCartDoorPushL))
  (Timeline|TL_CloseDoor ReverseFromEnd))

(event Custom|OnActiveAlternate
  (CallFunction|PlayAlternateOpeningSFXTemp)
  (bind (_direction _alpha) (Timeline|TL_CloseDoor_Alternate Play)
    (:Update
      (Variables|Default|SetDoorOperationPercentage _alpha))
    (:Finished
      (CallFunction|OnDoorTimelineFinished _direction))
    (:DestroyVehicles
      (switch Utilities|FlowControl|Switch|SwitchonETimelineDirection _direction
        (:Backward
          (Networking|SwitchHasAuthority
            (:Authority
              (CallFunction|SmashVehicles (CallFunction|GetVehiclestoSmash)))))))))

(event Custom|OnResetAlternate
  (CallFunction|PlayClosingSFXTemp)
  (Class|DragonCartDoorPushComponent|StartImpulseAsyncTargeting (Variables|Default|GetDragonCartDoorPushL))
  (Timeline|TL_CloseDoor_Alternate ReverseFromEnd))

(event OnActorTargeted(DragonCartDoorPushL)
  (params (TargetedActor :type Actor%20Object%20Reference))
  (CallFunction|PushAwayTarget TargetedActor))

(event OnActorTargeted(DragonCartDoorPushM)
  (params (TargetedActor :type Actor%20Object%20Reference))
  (CallFunction|PushAwayTarget TargetedActor))

(event OnActorTargeted(DragonCartDoorPushR)
  (params (TargetedActor :type Actor%20Object%20Reference))
  (CallFunction|PushAwayTarget TargetedActor))

(event Custom|PushAwayTarget
  (params (TargetedActor :type Actor%20Object%20Reference))
  (Networking|SwitchHasAuthority
    (:Authority
      (bind _asfort_player_pawn (Utilities|Casting|CastToFortPlayerPawn TargetedActor))
      (Character|LaunchCharacter _asfort_player_pawn (* (Transformation|GetActorRightVector) (ScalableFloat|GetValueatLevel (Variables|Default|GetHF%20Push%20Away%20Strength))) false false))))

(event Custom|PlayOpeningSFXTemp
  ;[1 1 1] The parent's GCs don't seem to play so adding this as a temporary measure to play sound.
  (GameplayCue|ExecuteGameplayCueOnActor(Burst) self (Variables|Default|GetGCDoorOpening) (Ability|GameplayCue|MakeGameplayCueParameters :NormalizedMagnitude 0.0 :RawMagnitude 0.0 :Location (Transformation|GetActorLocation self) :Normal "0,0,0" :GameplayEffectLevel 1 :AbilityLevel 1 :bReplicateLocationWhenUsingMinimalRepProxy false)))

(event Custom|PlayAlternateOpeningSFXTemp
  ;[1 1 1] The parent's GCs don't seem to play so adding this as a temporary measure to play sound.
  (GameplayCue|ExecuteGameplayCueOnActor(Burst) self (Variables|Default|GetGCDoorOpeningAlternate) (Ability|GameplayCue|MakeGameplayCueParameters :NormalizedMagnitude 0.0 :RawMagnitude 0.0 :Location (Transformation|GetActorLocation self) :Normal "0,0,0" :GameplayEffectLevel 1 :AbilityLevel 1 :bReplicateLocationWhenUsingMinimalRepProxy false)))

(event Custom|PlayClosingSFXTemp
  ;[1 1 1] The parent's GCs don't seem to play so adding this as a temporary measure to play sound.
  (GameplayCue|ExecuteGameplayCueOnActor(Burst) self (Variables|Default|GetGCDoorClosing) (Ability|GameplayCue|MakeGameplayCueParameters :NormalizedMagnitude 0.0 :RawMagnitude 0.0 :Location (Transformation|GetActorLocation self) :Normal "0,0,0" :GameplayEffectLevel 1 :AbilityLevel 1 :bReplicateLocationWhenUsingMinimalRepProxy false)))

(fn DetermineLaunchDirection
  (params (Actor :type Actor%20Object%20Reference))
  (Variables|Default|SetLaunchVelocity (Math|Vector|SelectVector (* (* (+ (Transformation|GetActorRightVector) (Math|Vector|MakeVector 0.0 0.0 (Variables|Default|GetLaunchBoostInitialZ))) (Variables|Default|GetLaunchPower)) (Math|Vector|MakeVector (* (.x (* (+ (Transformation|GetActorRightVector) (Math|Vector|MakeVector 0.0 0.0 (Variables|Default|GetLaunchBoostInitialZ))) (Variables|Default|GetLaunchPower))) (Variables|Default|GetLaunchBoostX)) (* (.y (* (+ (Transformation|GetActorRightVector) (Math|Vector|MakeVector 0.0 0.0 (Variables|Default|GetLaunchBoostInitialZ))) (Variables|Default|GetLaunchPower))) (Variables|Default|GetLaunchBoostY)) (* (.z (* (+ (Transformation|GetActorRightVector) (Math|Vector|MakeVector 0.0 0.0 (Variables|Default|GetLaunchBoostInitialZ))) (Variables|Default|GetLaunchPower))) (Variables|Default|GetLaunchBoostZ)))) (* (* (* (+ (Transformation|GetActorRightVector) (Math|Vector|MakeVector 0.0 0.0 (Variables|Default|GetLaunchBoostInitialZ))) (Variables|Default|GetLaunchPower)) (Math|Vector|MakeVector (* (.x (* (+ (Transformation|GetActorRightVector) (Math|Vector|MakeVector 0.0 0.0 (Variables|Default|GetLaunchBoostInitialZ))) (Variables|Default|GetLaunchPower))) (Variables|Default|GetLaunchBoostX)) (* (.y (* (+ (Transformation|GetActorRightVector) (Math|Vector|MakeVector 0.0 0.0 (Variables|Default|GetLaunchBoostInitialZ))) (Variables|Default|GetLaunchPower))) (Variables|Default|GetLaunchBoostY)) (* (.z (* (+ (Transformation|GetActorRightVector) (Math|Vector|MakeVector 0.0 0.0 (Variables|Default|GetLaunchBoostInitialZ))) (Variables|Default|GetLaunchPower))) (Variables|Default|GetLaunchBoostZ)))) "1.000000,-1.000000,1.000000") (<= (Math|Vector|Distance(Vector) (Transformation|GetActorLocation Actor) (Transformation|GetWorldLocation (Variables|Default|GetKillBox))) (Math|Vector|Distance(Vector) (Transformation|GetActorLocation Actor) (Transformation|GetWorldLocation (Variables|Default|GetKillBox))))))
  (return))

(fn EndVaultAlarm
  (for _array_element (Variables|Default|GetVaultAlarms)
    (Utilities|IsValid _array_element
      (:Is%20Valid
        (Class|BPILargeDoorTrigger|DeactivateVaultAlarms(Message) _array_element))))
  (return))

(fn GetPlayersToSmash
  (return (Collision|GetOverlappingActors (Variables|Default|GetKillBox) "/Script/FortniteGame.FortPlayerPawn" :!class /Script/Engine.PrimitiveComponent)))

(fn GetVehiclesToSmash
  (return (Collision|GetOverlappingActors (Variables|Default|GetKillBox) "/Script/FortniteGame.FortAthenaVehicle" :!class /Script/Engine.PrimitiveComponent)))

(fn MoveDoorByPercentage
  (params (Percentage :type Float%20%28double-precision%29))
  (|CallParentFunction Percentage)
  (Transformation|SetRelativeLocation (Variables|Default|GetRightDoorPanel) (Math|Vector|MakeVector (Math|Float|Lerp (Variables|Default|GetRightStartX) (Variables|Default|GetRightEndX) Percentage) (.y (Variables|/Script/Engine.SceneComponent|GetRelativeLocation (Variables|Default|GetRightDoorPanel))) (.z (Variables|/Script/Engine.SceneComponent|GetRelativeLocation (Variables|Default|GetRightDoorPanel)))) false false)
  (Transformation|SetRelativeLocation (Variables|Default|GetLeftDoorPanel) (Math|Vector|MakeVector (Math|Float|Lerp (Variables|Default|GetLeftStartX) (Variables|Default|GetLeftEndX) Percentage) (.y (Variables|/Script/Engine.SceneComponent|GetRelativeLocation (Variables|Default|GetLeftDoorPanel))) (.z (Variables|/Script/Engine.SceneComponent|GetRelativeLocation (Variables|Default|GetLeftDoorPanel)))) false false)
  (Transformation|SetRelativeLocation (Variables|Default|GetMiddleDoorPanel) (Math|Vector|MakeVector (.x (Variables|/Script/Engine.SceneComponent|GetRelativeLocation (Variables|Default|GetMiddleDoorPanel))) (.y (Variables|/Script/Engine.SceneComponent|GetRelativeLocation (Variables|Default|GetMiddleDoorPanel))) (Math|Float|Lerp (Variables|Default|GetMiddleStartZ) (Variables|Default|GetMiddleEndZ) Percentage)) false false))

(fn MovePanelsByPercentage
  (params (InputPin :type Float%20%28single-precision%29))
  (Transformation|SetRelativeLocation (Variables|Default|GetRightDoorPanel) (Math|Vector|MakeVector (Math|Float|Lerp (Variables|Default|GetRightStartX) (Variables|Default|GetRightEndX) InputPin) (.y (Variables|/Script/Engine.SceneComponent|GetRelativeLocation (Variables|Default|GetRightDoorPanel))) (.z (Variables|/Script/Engine.SceneComponent|GetRelativeLocation (Variables|Default|GetRightDoorPanel)))) false false)
  (Transformation|SetRelativeLocation (Variables|Default|GetLeftDoorPanel) (Math|Vector|MakeVector (Math|Float|Lerp (Variables|Default|GetLeftStartX) (Variables|Default|GetLeftEndX) InputPin) (.y (Variables|/Script/Engine.SceneComponent|GetRelativeLocation (Variables|Default|GetLeftDoorPanel))) (.z (Variables|/Script/Engine.SceneComponent|GetRelativeLocation (Variables|Default|GetLeftDoorPanel)))) false false)
  (Transformation|SetRelativeLocation (Variables|Default|GetMiddleDoorPanel) (Math|Vector|MakeVector (.x (Variables|/Script/Engine.SceneComponent|GetRelativeLocation (Variables|Default|GetMiddleDoorPanel))) (.y (Variables|/Script/Engine.SceneComponent|GetRelativeLocation (Variables|Default|GetMiddleDoorPanel))) (Math|Float|Lerp (Variables|Default|GetMiddleStartZ) (Variables|Default|GetMiddleEndZ) InputPin)) false false))

(fn OnDoorTimelineFinished
  (params (Direction :type ETimelineDirection%20Enum :default "Forward"))
  (bind _leftdoorpanel (Variables|Default|GetLeftDoorPanel))
  (|CallParentFunction Direction)
  (Class|DragonCartDoorPushComponent|StopImpulseAsyncTargeting (Variables|Default|GetDragonCartDoorPushL))
  (switch Utilities|FlowControl|Switch|SwitchonETimelineDirection Direction
    (:Forward
      (Collision|SetCollisionEnabled _leftdoorpanel "QueryOnly"))
    (:Backward
      (Collision|SetCollisionEnabled _leftdoorpanel "QueryAndPhysics")
      (Networking|SwitchHasAuthority
        (:Authority
          (CallFunction|SmashPlayers (CallFunction|GetPlayerstoSmash)))))))

(fn StartVaultAlarm
  (for _array_element (Variables|Default|GetVaultAlarms)
    (Utilities|IsValid _array_element
      (:Is%20Valid
        (Class|BPILargeDoorTrigger|ActivateVaultAlarms(Message) _array_element))))
  (return))

(fn ConstructionScript
  (Utilities|FlowControl|Sequence
    (:then_0
      (if (not (Variables|Default|GetBlockPartialOpen))
        (CallFunction|MovePanelsbyPercentage (/ (Variables|Default|GetPanelTravel) 1e+02))))
    (:then_1
      (Rendering|SetVisibility (Variables|Default|GetPositioningPlaneA) (Variables|Default|GetShowPositioningPlane) false))))
