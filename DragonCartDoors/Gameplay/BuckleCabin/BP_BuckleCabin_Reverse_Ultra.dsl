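; Child Blueprint of a three-panel sliding door (BuckleCabin reverse variant) with optional rotation.
; TL_CloseDoor drives the door percentage, playing from manual open, MegaVortex activation or delayed OnActive, and reversing on manual close.
; MoveDoorByPercentage lerps panel XY and optional yaw; the construction script previews partial opening from PanelTravel unless blocked.
; Manual opening fires sound and a GameplayCue; on forward finish, authority sets door state and removes the cue.
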(event EventBeginPlay
  (|CallParentFunction)
  (Variables|Default|SetBlockPartialOpen true)
  (Variables|Default|SetPanelTravel 0.0))

(event EventEndPlay
  (params (EndPlayReason :type EEndPlayReason%20Enum :default "Destroyed"))
  (|CallParentFunction EndPlayReason)
  (Variables|Default|SetBlockPartialOpen false)
  (Variables|Default|SetPanelTravel 0.0))

(event Custom|TempOpenCloseDemo
  ;[1 1 1] TEMP DEMO OPEN / CLOSE VISUALLY
  (Utilities|FlowControl|FlipFlop
    (:A
      (CallFunction|ManualDoorOpen))
    (:B
      (CallFunction|ManualDoorClose))))

(event Custom|ManualDoorOpen
  (bind _self self)
  (bind _returnvalue (Transformation|GetActorLocation _self))
  (GameplayCue|ExecuteGameplayCueOnActor(Burst) _self (Variables|Default|GetGCDoorsOpening) (Ability|GameplayCue|MakeGameplayCueParameters :NormalizedMagnitude 0.0 :RawMagnitude 0.0 :Location _returnvalue :Normal "0,0,0" :GameplayEffectLevel 1 :AbilityLevel 1 :bReplicateLocationWhenUsingMinimalRepProxy false))
  (Audio|PlaySoundatLocation "/Game/Sounds/Fort_Doors/World/Vault/Silo_Door_Gate_Open_Cue.Silo_Door_Gate_Open_Cue" _returnvalue "0, 0, 0" 2.0)
  (bind (_direction _alpha) (Timeline|TL_CloseDoor PlayFromStart)
    (:Update
      (Variables|Default|SetDoorOperationPercentage _alpha))
    (:Finished
      (CallFunction|OnDoorTimelineFinished _direction))))

(event EventBP_OnActive
  (params (EnableAlternateActive :type Boolean) (EnableOpenCues :type Boolean))
  (|CallParentFunction EnableAlternateActive EnableOpenCues)
  (Utilities|FlowControl|Delay (Variables|Default|GetCloseDelay))
  (Timeline|TL_CloseDoor Play))

(event Custom|ManualDoorClose
  (Timeline|TL_CloseDoor ReverseFromEnd))

(event Custom|Multi_MegaVortexActivated
  (CallFunction|ManualDoorOpen))

(fn MoveDoorByPercentage
  (params (Percentage :type Float%20%28double-precision%29))
  (|CallParentFunction Percentage)
  (Transformation|SetRelativeLocation (Variables|Default|GetRightDoorPanel) (Math|Vector|MakeVector (Math|Float|Lerp (Variables|Default|GetRightStartX) (Variables|Default|GetRightEndX) Percentage) (Math|Float|Lerp (Variables|Default|GetRightStartY) (Variables|Default|GetRightEndY) Percentage) (.z (Variables|/Script/Engine.SceneComponent|GetRelativeLocation (Variables|Default|GetRightDoorPanel)))) false false)
  (Transformation|SetRelativeLocation (Variables|Default|GetLeftDoorPanel) (Math|Vector|MakeVector (Math|Float|Lerp (Variables|Default|GetLeftStartX) (Variables|Default|GetLeftEndX) Percentage) (Math|Float|Lerp (Variables|Default|GetLeftStartY) (Variables|Default|GetLeftEndY) Percentage) (.z (Variables|/Script/Engine.SceneComponent|GetRelativeLocation (Variables|Default|GetLeftDoorPanel)))) false false)
  (Transformation|SetRelativeLocation (Variables|Default|GetMiddleDoorPanel) (Math|Vector|MakeVector (Math|Float|Lerp (Variables|Default|GetMiddleStartX) (Variables|Default|GetMiddleEndX) Percentage) (Math|Float|Lerp (Variables|Default|GetMiddleStartY) (Variables|Default|GetMiddleEndY) Percentage) (.z (Variables|/Script/Engine.SceneComponent|GetRelativeLocation (Variables|Default|GetMiddleDoorPanel)))) false false)
  (if (Variables|Default|GetRotateDoorOpening)
    (Transformation|SetRelativeRotation (Variables|Default|GetScene) (Math|Rotator|MakeRotator (.roll (Variables|/Script/Engine.SceneComponent|GetRelativeRotation (Variables|Default|GetScene))) (.pitch (Variables|/Script/Engine.SceneComponent|GetRelativeRotation (Variables|Default|GetScene))) (Math|Float|Lerp (Variables|Default|GetRotationStartZ) (Variables|Default|GetRotationEndZ) Percentage)) false false)))

(fn OnDoorTimelineFinished
  (params (Direction :type ETimelineDirection%20Enum :default "Forward"))
  (|CallParentFunction Direction)
  (switch Utilities|FlowControl|Switch|SwitchonETimelineDirection Direction
    (:Forward
      (Networking|SwitchHasAuthority
        (:Authority
          (Variables|Default|SetDoorState "NewEnumerator2")
          (GameplayCue|RemoveGameplayCueOnActor(Looping) self (Variables|Default|GetGCDoorsOpening) (Ability|GameplayCue|MakeGameplayCueParameters :NormalizedMagnitude 0.0 :RawMagnitude 0.0 :Location (Transformation|GetActorLocation self) :Normal "0,0,0" :GameplayEffectLevel 1 :AbilityLevel 1 :bReplicateLocationWhenUsingMinimalRepProxy false)))))))

(fn ConstructionScript
  (bind _paneltravel (Variables|Default|GetPanelTravel))
  (bind _returnvalue (/ _paneltravel 1e+02))
  (Utilities|FlowControl|Sequence
    (:then_0
      (Variables|Default|SetRotationEndZ (Variables|Default|GetDoorOpeningRotation))
      (if (not (Variables|Default|GetBlockPartialOpen))
        (Transformation|SetRelativeLocation (Variables|Default|GetRightDoorPanel) (Math|Vector|MakeVector (Math|Float|Lerp (Variables|Default|GetRightStartX) (Variables|Default|GetRightEndX) _returnvalue) (Math|Float|Lerp (Variables|Default|GetRightStartY) (Variables|Default|GetRightEndY) _returnvalue) (.z (Variables|/Script/Engine.SceneComponent|GetRelativeLocation (Variables|Default|GetRightDoorPanel)))) false false)
        (Transformation|SetRelativeLocation (Variables|Default|GetLeftDoorPanel) (Math|Vector|MakeVector (Math|Float|Lerp (Variables|Default|GetLeftStartX) (Variables|Default|GetLeftEndX) _returnvalue) (Math|Float|Lerp (Variables|Default|GetLeftStartY) (Variables|Default|GetLeftEndY) _returnvalue) (.z (Variables|/Script/Engine.SceneComponent|GetRelativeLocation (Variables|Default|GetLeftDoorPanel)))) false false)
        (Transformation|SetRelativeLocation (Variables|Default|GetMiddleDoorPanel) (Math|Vector|MakeVector (Math|Float|Lerp (Variables|Default|GetMiddleStartX) (Variables|Default|GetMiddleEndX) _returnvalue) (Math|Float|Lerp (Variables|Default|GetMiddleStartY) (Variables|Default|GetMiddleEndY) _returnvalue) (.z (Variables|/Script/Engine.SceneComponent|GetRelativeLocation (Variables|Default|GetMiddleDoorPanel)))) false false)
        (if (Variables|Default|GetRotateDoorOpening)
          (Transformation|SetRelativeRotation (Variables|Default|GetScene) (Math|Rotator|MakeRotator (.roll (Variables|/Script/Engine.SceneComponent|GetRelativeRotation (Variables|Default|GetScene))) (.pitch (Variables|/Script/Engine.SceneComponent|GetRelativeRotation (Variables|Default|GetScene))) (Math|Float|Lerp (Variables|Default|GetRotationStartZ) (Variables|Default|GetRotationEndZ) _returnvalue)) false false))))
    (:then_1
      (Rendering|SetVisibility (Variables|Default|GetPositioningSphere) (Variables|Default|GetShowPositioningSphere) false))))
