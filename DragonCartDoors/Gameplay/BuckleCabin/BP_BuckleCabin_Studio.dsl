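; Two-panel DragonCart door (BuckleCabin Studio) that slides shut on activation, pushing players aside and smashing vehicles and players caught in its kill box.
; TL_CloseDoor drives the door percentage, which lerps panel position and scale; reset plays it in reverse.
; Server handles launches and smashes; push targeting starts only on the left component, though both are bound.
; Destruction drops metal and spawns sound and Niagara debris; construction previews panel travel, and an optional temporary demo timer cycles the door.
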(event EventBeginPlay
  (|CallParentFunction)
  (Variables|Default|SetBlockPartialOpen true)
  (Variables|Default|SetPanelTravel 0.0)
  (Variables|/Script/DragonCartDoorsRuntime.DragonCartDoorPushComponent|SetPushPointSceneComp (Variables|Default|GetPushPoint_L) (Variables|Default|GetDragonCartDoorPush_L))
  (Variables|/Script/DragonCartDoorsRuntime.DragonCartDoorPushComponent|SetPushPointSceneComp (Variables|Default|GetPushPoint_R) (Variables|Default|GetDragonCartDoorPush_R))
  (bind _actorcenterpoint (CallFunction|GetActorCenterPoint))
  (Variables|Default|SetActorCenterLocation _actorcenterpoint)
  ;[1 1 1] TEMP DEMO OPEN / CLOSE VISUALLY
  (if (Variables|Default|GetEnableDemo)
    (Utilities|Time|SetTimerbyEvent (delegate Custom|TempOpenCloseDemo) 15.0 true)))

(event EventEndPlay
  (params (EndPlayReason :type EEndPlayReason%20Enum :default "Destroyed"))
  (|CallParentFunction EndPlayReason)
  (Variables|Default|SetBlockPartialOpen false)
  (Variables|Default|SetPanelTravel 0.0))

(event Custom|TempOpenCloseDemo
  ;[1 1 1] TEMP DEMO OPEN / CLOSE VISUALLY
  (Utilities|FlowControl|FlipFlop
    (:A
      (CallFunction|TempCloseDemo))
    (:B
      (CallFunction|TempOpenDemo))))

(event EventSetDoorHealthFromDataTable
  (Gameplay|ForceBuildingHealth self (Math|Conversions|ToFloat(Integer) (Math|Float|Truncate (ScalableFloat|GetValueatLevel (Variables|Default|GetScalarDoorHealth))))))

(event Custom|TempCloseDemo
  (bind (_direction _alpha) (Timeline|TL_CloseDoor PlayFromStart)
    (:Update
      (Variables|Default|SetDoorOperationPercentage _alpha))
    (:Finished
      (CallFunction|OnDoorTimelineFinished _direction))
    (:DestroyVehicles
      (switch Utilities|FlowControl|Switch|SwitchonETimelineDirection _direction
        (:Forward
          (Networking|SwitchHasAuthority
            (:Authority
              (CallFunction|SmashVehicles (CallFunction|GetVehiclestoSmash)))))))))

(event EventBP_OnActive
  (params (EnableAlternateActive :type Boolean) (EnableOpenCues :type Boolean))
  (|CallParentFunction EnableAlternateActive EnableOpenCues)
  (Class|DragonCartDoorPushComponent|StartImpulseAsyncTargeting (Variables|Default|GetDragonCartDoorPush_L))
  (Timeline|TL_CloseDoor PlayFromStart))

(event Custom|TempOpenDemo
  (Timeline|TL_CloseDoor ReverseFromEnd))

(event Default|EventBP_OnReset
  (params (EnableAlternateReset :type Boolean) (EnableOpenCues :type Boolean))
  (|CallParentFunction EnableAlternateReset EnableOpenCues)
  (Timeline|TL_CloseDoor ReverseFromEnd))

(event EventBP_OnClosed
  (params (EnableAlternateClose :type Boolean) (EnableOpenCues :type Boolean))
  (|CallParentFunction EnableAlternateClose EnableOpenCues))

(event EventOnDeathServer
  (params (Damage :type Float%20%28single-precision%29) (DamageTags :type Gameplay%20Tag%20Container%20Structure%20%28by%20ref%29) (Momentum :type Vector :default "0, 0, 0") (HitInfo :type Hit%20Result%20Structure%20%28by%20ref%29) (InstigatedBy :type Controller%20Object%20Reference) (DamageCauser :type Actor%20Object%20Reference) (EffectContext :type Gameplay%20Effect%20Context%20Handle%20Structure))
  (|CallParentFunction Damage DamageTags Momentum HitInfo InstigatedBy DamageCauser EffectContext)
  (Inventory|SpawnPickupInWorld "/Game/Items/ResourcePickups/MetalItemData.MetalItemData" 24 (Variables|Default|GetActorCenterLocation) "0.000000,0.000000,1.000000" 24 true true true -1 "Destruction" "Unset" 0 false))

(event EventOnDeathPlayEffects
  (params (Damage :type Float%20%28single-precision%29) (DamageTags :type Gameplay%20Tag%20Container%20Structure%20%28by%20ref%29) (Momentum :type Vector :default "0, 0, 0") (HitInfo :type Hit%20Result%20Structure%20%28by%20ref%29) (InstigatedBy :type Fort%20Pawn%20Object%20Reference) (DamageCauser :type Actor%20Object%20Reference) (EffectContext :type Gameplay%20Effect%20Context%20Handle%20Structure))
  (bind (_bblockinghit _binitialoverlap _time _distance _location) (Collision|BreakHitResult HitInfo))
  (bind _panel_r (Variables|Default|GetPanel_R))
  (bind (_returnvalue_origin _returnvalue_boxextent _returnvalue_sphereradius) (Math|BoxSphereBounds|BreakBoxSphereBounds (Fortnite|Utility|GetLocalBounds _panel_r)))
  (bind _panel_l (Variables|Default|GetPanel_L))
  (bind (_returnvalue_origin_2 _returnvalue_boxextent_2 _returnvalue_sphereradius_2) (Math|BoxSphereBounds|BreakBoxSphereBounds (Fortnite|Utility|GetLocalBounds _panel_l)))
  (bind (_returnvalue_location _returnvalue_rotation) (Math|Transform|BreakTransform (Transformation|GetActorTransform)))
  (Audio|PlaySoundatLocation (Variables|Default|GetDestroySound) _location)
  (bind _returnvalue (Niagara|SpawnSystematLocation "/Game/Effects/Fort_Effects/World/Destruction/Destruction2/Custom/NS_LargeDoor_Destruction.NS_LargeDoor_Destruction" (Transformation|GetActorLocation)))
  (Niagara|SetNiagaraVariable(Vector3) _returnvalue "DoorScaling" (Variables|/Script/Engine.SceneComponent|GetRelativeScale3D (Variables|Default|GetPanel_L)))
  (Niagara|SetNiagaraVariable(Vector3) _returnvalue "HitVector" (Math|Vector|Normalize (- _location (Transformation|GetActorLocation InstigatedBy))))
  (Niagara|SetNiagaraVariable(Vector3) _returnvalue "HitLocation" _location)
  (Niagara|SetNiagaraVariable(Vector3) _returnvalue "BoundsExtents" (+ _returnvalue_boxextent_2 _returnvalue_boxextent))
  (Niagara|SetNiagaraVariable(Float) _returnvalue "Radius" (+ _returnvalue_sphereradius_2 _returnvalue_sphereradius))
  (Niagara|SetNiagaraVariable(Quaternion) _returnvalue "Orientation" (Math|Conversions|ToQuaternion(Rotator) _returnvalue_rotation)))

(event OnLinkedActorDestroyed(FortLinkToActor)
  (params (DamageCauser :type Actor%20Object%20Reference))
  (Rendering|SetActorHiddenInGame true)
  (Actor|DestroyActor))

(event OnActorTargeted(DragonCartDoorPush_L)
  (params (TargetedActor :type Actor%20Object%20Reference))
  (CallFunction|PushTargetedPlayers TargetedActor))

(event OnActorTargeted(DragonCartDoorPush_R)
  (params (TargetedActor :type Actor%20Object%20Reference))
  (CallFunction|PushTargetedPlayers TargetedActor))

(event Custom|PushTargetedPlayers
  (params (Actor :type Actor%20Object%20Reference))
  (Networking|SwitchHasAuthority
    (:Authority
      (bind _asfort_player_pawn (Utilities|Casting|CastToFortPlayerPawn Actor))
      (Character|LaunchCharacter _asfort_player_pawn (* (Transformation|GetActorRightVector) (ScalableFloat|GetValueatLevel (Variables|Default|GetHF_PushAwayStrength))) false false))))

(fn DetermineLaunchDirection
  (params (Actor :type Actor%20Object%20Reference))
  (Variables|Default|SetLaunchVelocity (Math|Vector|SelectVector (* (* (+ (Transformation|GetActorRightVector) (Math|Vector|MakeVector 0.0 0.0 (Variables|Default|GetLaunchBoostInitialZ))) (Variables|Default|GetLaunchPower)) (Math|Vector|MakeVector (* (.x (* (+ (Transformation|GetActorRightVector) (Math|Vector|MakeVector 0.0 0.0 (Variables|Default|GetLaunchBoostInitialZ))) (Variables|Default|GetLaunchPower))) (Variables|Default|GetLaunchBoostX)) (* (.y (* (+ (Transformation|GetActorRightVector) (Math|Vector|MakeVector 0.0 0.0 (Variables|Default|GetLaunchBoostInitialZ))) (Variables|Default|GetLaunchPower))) (Variables|Default|GetLaunchBoostY)) (* (.z (* (+ (Transformation|GetActorRightVector) (Math|Vector|MakeVector 0.0 0.0 (Variables|Default|GetLaunchBoostInitialZ))) (Variables|Default|GetLaunchPower))) (Variables|Default|GetLaunchBoostZ)))) (* (* (* (+ (Transformation|GetActorRightVector) (Math|Vector|MakeVector 0.0 0.0 (Variables|Default|GetLaunchBoostInitialZ))) (Variables|Default|GetLaunchPower)) (Math|Vector|MakeVector (* (.x (* (+ (Transformation|GetActorRightVector) (Math|Vector|MakeVector 0.0 0.0 (Variables|Default|GetLaunchBoostInitialZ))) (Variables|Default|GetLaunchPower))) (Variables|Default|GetLaunchBoostX)) (* (.y (* (+ (Transformation|GetActorRightVector) (Math|Vector|MakeVector 0.0 0.0 (Variables|Default|GetLaunchBoostInitialZ))) (Variables|Default|GetLaunchPower))) (Variables|Default|GetLaunchBoostY)) (* (.z (* (+ (Transformation|GetActorRightVector) (Math|Vector|MakeVector 0.0 0.0 (Variables|Default|GetLaunchBoostInitialZ))) (Variables|Default|GetLaunchPower))) (Variables|Default|GetLaunchBoostZ)))) "1.000000,-1.000000,1.000000") (<= (Math|Vector|Distance(Vector) (Transformation|GetActorLocation Actor) (Transformation|GetWorldLocation (Variables|Default|GetKillBox))) (Math|Vector|Distance(Vector) (Transformation|GetActorLocation Actor) (Transformation|GetWorldLocation (Variables|Default|GetKillBox))))))
  (return))

(fn GetActorCenterPoint
  (bind (_origin _boxextent) (Collision|GetActorBounds false))
  (return (Math|Vector|MakeVector (.x (Transformation|GetActorLocation)) (.y (Transformation|GetActorLocation)) (+ (.z (Transformation|GetActorLocation)) (* (.z _boxextent) 0.5)))))

(fn GetPlayersToSmash
  (return (Collision|GetOverlappingActors (Variables|Default|GetKillBox) "/Script/FortniteGame.FortPlayerPawn" :!class /Script/Engine.PrimitiveComponent)))

(fn GetVehiclesToSmash
  (return (Collision|GetOverlappingActors (Variables|Default|GetKillBox) "/Script/FortniteGame.FortAthenaVehicle" :!class /Script/Engine.PrimitiveComponent)))

(fn MoveDoorByPercentage
  (params (Percentage :type Float%20%28double-precision%29))
  (|CallParentFunction Percentage)
  (Transformation|SetRelativeLocation (Variables|Default|GetPanel_L) (Math|Vector|MakeVector (Math|Float|Lerp (Variables|Default|GetLeftStartX) (Variables|Default|GetLeftEndX) Percentage) (.y (Variables|/Script/Engine.SceneComponent|GetRelativeLocation (Variables|Default|GetPanel_L))) (.z (Variables|/Script/Engine.SceneComponent|GetRelativeLocation (Variables|Default|GetPanel_L)))) false false)
  (Transformation|SetRelativeLocation (Variables|Default|GetPanel_R) (Math|Vector|MakeVector (Math|Float|Lerp (Variables|Default|GetRightStartX) (Variables|Default|GetRightEndX) Percentage) (.y (Variables|/Script/Engine.SceneComponent|GetRelativeLocation (Variables|Default|GetPanel_L))) (.z (Variables|/Script/Engine.SceneComponent|GetRelativeLocation (Variables|Default|GetPanel_L)))) false false)
  (Transformation|SetRelativeScale3D (Variables|Default|GetPanel_L) (Math|Vector|Lerp(Vector) (Variables|Default|GetLeftStartScaleX) (Variables|Default|GetLeftEndScaleX) Percentage))
  (Transformation|SetRelativeScale3D (Variables|Default|GetPanel_R) (Math|Vector|Lerp(Vector) (Variables|Default|GetRightStartScaleX) (Variables|Default|GetRightEndScaleX) Percentage)))

(fn OnDoorTimelineFinished
  (params (Direction :type ETimelineDirection%20Enum :default "Forward"))
  (|CallParentFunction Direction)
  (Class|DragonCartDoorPushComponent|StopImpulseAsyncTargeting (Variables|Default|GetDragonCartDoorPush_L))
  (switch Utilities|FlowControl|Switch|SwitchonETimelineDirection Direction
    (:Forward
      (Networking|SwitchHasAuthority
        (:Authority
          (CallFunction|SmashPlayers (CallFunction|GetPlayerstoSmash)))))))

(fn ConstructionScript
  (bind _paneltravel (Variables|Default|GetPanelTravel))
  (bind _returnvalue (/ _paneltravel 1e+02))
  (Utilities|FlowControl|Sequence
    (:then_0
      (if (not (Variables|Default|GetBlockPartialOpen))
        (Transformation|SetRelativeLocation (Variables|Default|GetPanel_L) (Math|Vector|MakeVector (Math|Float|Lerp (Variables|Default|GetLeftStartX) (Variables|Default|GetLeftEndX) _returnvalue) (.y (Variables|/Script/Engine.SceneComponent|GetRelativeLocation (Variables|Default|GetPanel_L))) (.z (Variables|/Script/Engine.SceneComponent|GetRelativeLocation (Variables|Default|GetPanel_L)))) false false)
        (Transformation|SetRelativeLocation (Variables|Default|GetPanel_R) (Math|Vector|MakeVector (Math|Float|Lerp (Variables|Default|GetRightStartX) (Variables|Default|GetRightEndX) _returnvalue) (.y (Variables|/Script/Engine.SceneComponent|GetRelativeLocation (Variables|Default|GetPanel_R))) (.z (Variables|/Script/Engine.SceneComponent|GetRelativeLocation (Variables|Default|GetPanel_R)))) false false)
        (Transformation|SetRelativeScale3D (Variables|Default|GetPanel_L) (Math|Vector|Lerp(Vector) (Variables|Default|GetLeftStartScaleX) (Variables|Default|GetLeftEndScaleX) _returnvalue))
        (Transformation|SetRelativeScale3D (Variables|Default|GetPanel_R) (Math|Vector|Lerp(Vector) (Variables|Default|GetRightStartScaleX) (Variables|Default|GetRightEndScaleX) _returnvalue))))
    (:then_1
      (Rendering|SetVisibility (Variables|Default|GetPositioningPlaneA) (Variables|Default|GetShowPositioningPlane) false))))
