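; Multi-panel vault door, a child door Blueprint that animates bottom, middle and side panels in staggered phases from one percentage.
; Activation starts registered alarms if enabled, fires steam VFX and audio cue, then plays TL_CloseDoor, which drives the percentage.
; Near the timeline's end it hides the middle panel and stops alarms.
; Construction script previews partial opening via PanelTravel; BeginPlay holds a temporary VFX test loop, and push/smash events are empty stubs.
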(event EventBeginPlay
  (|CallParentFunction)
  (Variables|Default|SetBlockPartialOpen true)
  (Variables|Default|SetPanelTravel 0.0)
  ;[1 1 1] TEMP SPAGHETTI TO HELP VFX, REMOVE ME!
  (if (Variables|Default|GetbTestRepeatingOpen)
    (Utilities|FlowControl|Delay 0.0)
    (Development|PrintString "Test open")
    (Components|Activation|Activate (Variables|Default|GetOpeningSteamNS) false)
    (GameplayCue|ExecuteGameplayCueOnActor(Burst) self (Variables|Default|GetGCVaultDoorSteamAudio) (Ability|GameplayCue|MakeGameplayCueParameters :NormalizedMagnitude 0.0 :RawMagnitude 0.0 :Location (Transformation|GetActorLocation self) :Normal "0,0,0" :GameplayEffectLevel 1 :AbilityLevel 1 :bReplicateLocationWhenUsingMinimalRepProxy false))
    (Utilities|FlowControl|Delay 2.0)
    (bind (_direction _alpha) (Timeline|Timeline PlayFromStart)
      (:Update
        _)
      (:Finished
        (Utilities|FlowControl|FlipFlop
          (:A
            (Development|PrintString "Test close")
            (Timeline|Timeline ReverseFromEnd))
          (:B
            _))))
    (Variables|Default|SetDoorOperationPercentage _alpha)))

(event EventEndPlay
  (params (EndPlayReason :type EEndPlayReason%20Enum :default "Destroyed"))
  (|CallParentFunction EndPlayReason)
  (Variables|Default|SetBlockPartialOpen false)
  (Variables|Default|SetPanelTravel 0.0))

(event EventRegisterVaultDoorAlarm
  (params (AlarmToRegister :type Actor%20Object%20Reference))
  (Utilities|Array|AddUnique (Variables|Default|GetVaultAlarms) AlarmToRegister))

(event EventBP_OnOpen
  (params (EnableAlternateOpen :type Boolean) (EnableOpenCues :type Boolean))
  (|CallParentFunction EnableAlternateOpen EnableOpenCues))

(event EventBP_OnActive
  (params (EnableAlternateActive :type Boolean) (EnableOpenCues :type Boolean))
  (|CallParentFunction EnableAlternateActive EnableOpenCues)
  (Variables|Default|SetEnableAlternateOpen EnableAlternateActive)
  (if (Variables|Default|GetPlayAlarm)
    (CallFunction|StartVaultAlarm)
    (else
      _))
  (Utilities|FlowControl|Delay (Math|Float|SelectFloat 0.0 (ScalableFloat|GetValueasDouble (Variables|Default|GetActivationDelay)) (Variables|Default|GetEnableAlternateOpen)))
  (Components|Activation|Activate (Variables|Default|GetOpeningSteamNS) false)
  (GameplayCue|ExecuteGameplayCueOnActor(Burst) self (Variables|Default|GetGCVaultDoorSteamAudio) (Ability|GameplayCue|MakeGameplayCueParameters :NormalizedMagnitude 0.0 :RawMagnitude 0.0 :Location (Transformation|GetActorLocation self) :Normal "0,0,0" :GameplayEffectLevel 1 :AbilityLevel 1 :bReplicateLocationWhenUsingMinimalRepProxy false))
  (Utilities|FlowControl|Delay 1.0)
  (if (Variables|Default|GetEnableAlternateOpen)
    (CallFunction|OnActiveAlternate)
    (Development|PrintString "Large Door - Alternate Open" true true "(R=1.000000,G=0.400000,B=0.000000,A=1.000000)" 5.0)
    (else
      (CallFunction|OnActiveNormal)
      (Development|PrintString "Large Door - Normal Open" true true "(R=1.000000,G=0.400000,B=0.000000,A=1.000000)" 5.0))))

(event Custom|PushPlayersAway)

(event Custom|SmashPlayer)

(event Custom|SmashCar)

(event Custom|OnActiveNormal
  (bind (_direction _alpha) (Timeline|TL_CloseDoor Play)
    (:Update
      (Variables|Default|SetDoorOperationPercentage _alpha))
    (:Finished)
    (:DoorAlmostClosed
      (Development|SetHiddeninGame (Variables|Default|GetMiddleDoorPanel) true false)
      (Collision|SetCollisionEnabled (Variables|Default|GetMiddleDoorPanel) "NoCollision")
      (Default|EndVaultAlarm))))

(event Custom|OnActiveAlternate
  (Timeline|TL_CloseDoor Play))

(event EventOnDeathPlayEffects
  (params (Damage :type Float%20%28single-precision%29) (DamageTags :type Gameplay%20Tag%20Container%20Structure%20%28by%20ref%29) (Momentum :type Vector :default "0, 0, 0") (HitInfo :type Hit%20Result%20Structure%20%28by%20ref%29) (InstigatedBy :type Fort%20Pawn%20Object%20Reference) (DamageCauser :type Actor%20Object%20Reference) (EffectContext :type Gameplay%20Effect%20Context%20Handle%20Structure))
  (bind (_bblockinghit _binitialoverlap _time _distance _location) (Collision|BreakHitResult HitInfo))
  (Audio|PlaySoundatLocation (Variables|Default|GetDestroy%20Sound) _location))

(fn DetermineLaunchDirection
  (params (Actor :type Actor%20Object%20Reference))
  (Variables|Default|SetLaunchVelocity (Math|Vector|SelectVector (* (* (+ (Transformation|GetActorRightVector) (Math|Vector|MakeVector 0.0 0.0 (Variables|Default|GetLaunchBoostInitialZ))) (Variables|Default|GetLaunchPower)) (Math|Vector|MakeVector (* (.x (* (+ (Transformation|GetActorRightVector) (Math|Vector|MakeVector 0.0 0.0 (Variables|Default|GetLaunchBoostInitialZ))) (Variables|Default|GetLaunchPower))) (Variables|Default|GetLaunchBoostX)) (* (.y (* (+ (Transformation|GetActorRightVector) (Math|Vector|MakeVector 0.0 0.0 (Variables|Default|GetLaunchBoostInitialZ))) (Variables|Default|GetLaunchPower))) (Variables|Default|GetLaunchBoostY)) (* (.z (* (+ (Transformation|GetActorRightVector) (Math|Vector|MakeVector 0.0 0.0 (Variables|Default|GetLaunchBoostInitialZ))) (Variables|Default|GetLaunchPower))) (Variables|Default|GetLaunchBoostZ)))) (* (* (* (+ (Transformation|GetActorRightVector) (Math|Vector|MakeVector 0.0 0.0 (Variables|Default|GetLaunchBoostInitialZ))) (Variables|Default|GetLaunchPower)) (Math|Vector|MakeVector (* (.x (* (+ (Transformation|GetActorRightVector) (Math|Vector|MakeVector 0.0 0.0 (Variables|Default|GetLaunchBoostInitialZ))) (Variables|Default|GetLaunchPower))) (Variables|Default|GetLaunchBoostX)) (* (.y (* (+ (Transformation|GetActorRightVector) (Math|Vector|MakeVector 0.0 0.0 (Variables|Default|GetLaunchBoostInitialZ))) (Variables|Default|GetLaunchPower))) (Variables|Default|GetLaunchBoostY)) (* (.z (* (+ (Transformation|GetActorRightVector) (Math|Vector|MakeVector 0.0 0.0 (Variables|Default|GetLaunchBoostInitialZ))) (Variables|Default|GetLaunchPower))) (Variables|Default|GetLaunchBoostZ)))) "1.000000,-1.000000,1.000000") (<= (Math|Vector|Distance(Vector) (Transformation|GetActorLocation Actor) (Transformation|GetWorldLocation (Variables|Default|GetKillBox))) (Math|Vector|Distance(Vector) (Transformation|GetActorLocation Actor) (Transformation|GetWorldLocation (Variables|Default|GetKillBox))))))
  (return))

(fn EndVaultAlarm
  (for _array_element (Variables|Default|GetVaultAlarms)
    (Utilities|IsValid _array_element
      (:Is%20Valid
        (Class|BPILargeDoorTrigger|DeactivateVaultAlarms(Message) _array_element))))
  (return))

(fn MoveDoorByPercentage
  (params (Percentage :type Float%20%28double-precision%29))
  (bind _returnvalue (Math|Float|NormalizetoRange Percentage 0.25 1.0))
  (bind _returnvalue_2 (Math|Float|Clamp(Float) _returnvalue))
  (|CallParentFunction Percentage)
  (Transformation|SetRelativeLocation (Variables|Default|GetBottomDoorPanel) (Math|Vector|MakeVector (.x (Variables|/Script/Engine.SceneComponent|GetRelativeLocation (Variables|Default|GetBottomDoorPanel))) (.y (Variables|/Script/Engine.SceneComponent|GetRelativeLocation (Variables|Default|GetBottomDoorPanel))) (Math|Float|Lerp (Variables|Default|GetBottomStartZ) (Variables|Default|GetBottomEndZ) (Math|Float|Clamp(Float) (Math|Float|NormalizetoRange Percentage 0.0 0.15)))) false false)
  (Transformation|SetRelativeLocation (Variables|Default|GetMiddleDoorPanel) (Math|Vector|MakeVector (.x (Variables|/Script/Engine.SceneComponent|GetRelativeLocation (Variables|Default|GetMiddleDoorPanel))) (.y (Variables|/Script/Engine.SceneComponent|GetRelativeLocation (Variables|Default|GetMiddleDoorPanel))) (Math|Float|Lerp (Variables|Default|GetMiddleStartZ) (Variables|Default|GetMiddleEndZ) (Math|Float|Clamp(Float) (Math|Float|NormalizetoRange Percentage 0.15 1.0)))) false false)
  (Transformation|SetRelativeLocation (Variables|Default|GetRightDoorPanel) (Math|Vector|MakeVector (Math|Float|Lerp (Variables|Default|GetRightStartX) (Variables|Default|GetRightEndX) _returnvalue_2) (.y (Variables|/Script/Engine.SceneComponent|GetRelativeLocation (Variables|Default|GetRightDoorPanel))) (.z (Variables|/Script/Engine.SceneComponent|GetRelativeLocation (Variables|Default|GetRightDoorPanel)))) false false)
  (Transformation|SetRelativeLocation (Variables|Default|GetLeftDoorPanel) (Math|Vector|MakeVector (Math|Float|Lerp (Variables|Default|GetLeftStartX) (Variables|Default|GetLeftEndX) _returnvalue_2) (.y (Variables|/Script/Engine.SceneComponent|GetRelativeLocation (Variables|Default|GetLeftDoorPanel))) (.z (Variables|/Script/Engine.SceneComponent|GetRelativeLocation (Variables|Default|GetLeftDoorPanel)))) false false))

(fn MovePanelsByPercentage
  (params (InputAlpha1 :type Float%20%28single-precision%29) (InputAlpha2 :type Float%20%28double-precision%29) (InputAlpha3 :type Float%20%28double-precision%29))
  (Transformation|SetRelativeLocation (Variables|Default|GetRightDoorPanel) (Math|Vector|MakeVector (Math|Float|Lerp (Variables|Default|GetRightStartX) (Variables|Default|GetRightEndX) InputAlpha3) (.y (Variables|/Script/Engine.SceneComponent|GetRelativeLocation (Variables|Default|GetRightDoorPanel))) (.z (Variables|/Script/Engine.SceneComponent|GetRelativeLocation (Variables|Default|GetRightDoorPanel)))) false false)
  (Transformation|SetRelativeLocation (Variables|Default|GetLeftDoorPanel) (Math|Vector|MakeVector (Math|Float|Lerp (Variables|Default|GetLeftStartX) (Variables|Default|GetLeftEndX) InputAlpha3) (.y (Variables|/Script/Engine.SceneComponent|GetRelativeLocation (Variables|Default|GetLeftDoorPanel))) (.z (Variables|/Script/Engine.SceneComponent|GetRelativeLocation (Variables|Default|GetLeftDoorPanel)))) false false)
  (Transformation|SetRelativeLocation (Variables|Default|GetMiddleDoorPanel) (Math|Vector|MakeVector (.x (Variables|/Script/Engine.SceneComponent|GetRelativeLocation (Variables|Default|GetMiddleDoorPanel))) (.y (Variables|/Script/Engine.SceneComponent|GetRelativeLocation (Variables|Default|GetMiddleDoorPanel))) (Math|Float|Lerp (Variables|Default|GetMiddleStartZ) (Variables|Default|GetMiddleEndZ) InputAlpha2)) false false)
  (Transformation|SetRelativeLocation (Variables|Default|GetBottomDoorPanel) (Math|Vector|MakeVector (.x (Variables|/Script/Engine.SceneComponent|GetRelativeLocation (Variables|Default|GetBottomDoorPanel))) (.y (Variables|/Script/Engine.SceneComponent|GetRelativeLocation (Variables|Default|GetBottomDoorPanel))) (Math|Float|Lerp (Variables|Default|GetBottomStartZ) (Variables|Default|GetBottomEndZ) InputAlpha1)) false false))

(fn StartVaultAlarm
  (for _array_element (Variables|Default|GetVaultAlarms)
    (Utilities|IsValid _array_element
      (:Is%20Valid
        (Class|BPILargeDoorTrigger|ActivateVaultAlarms(Message) _array_element))))
  (return))

(fn ConstructionScript
  (bind _paneltravel (Variables|Default|GetPanelTravel))
  (bind _returnvalue (/ _paneltravel 1e+02))
  (Utilities|FlowControl|Sequence
    (:then_0
      (if (not (Variables|Default|GetBlockPartialOpen))
        (bind _output_get (LocalVariables|SetPanelTravelStep1 (* _returnvalue 15.0)))
        (bind _output_get_2 (LocalVariables|SetPanelTravelStep2 (- _output_get 2.0)))
        (bind _output_get_3 (LocalVariables|SetPanelTravelStep3 (- _output_get 2.5)))
        (CallFunction|MovePanelsbyPercentage (* _returnvalue 3.0) (Math|Float|Clamp(Float) (/ _output_get_2 13.0)) (Math|Float|Clamp(Float) (/ _output_get_3 12.5)))))
    (:then_1
      (Rendering|SetVisibility (Variables|Default|GetPositioningPlaneA) (Variables|Default|GetShowPositioningPlane) false))))
